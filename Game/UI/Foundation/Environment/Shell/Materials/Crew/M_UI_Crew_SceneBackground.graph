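# Two-sided lit material for the Crew scene background in the UI shell, with blue and gold gradient palettes and an Is Subscribed parameter.
# An exponential radial mask with adjustable offset shapes outer brightness.
# Screen-fitted, rotatable UVs drive two panning samples of the Crew pattern, merged by max and smoothstep, then added to or subtracted from the gradient.
# The result feeds the shared MPItemShop setup function, which also yields a floor mask.
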Material /Game/UI/Foundation/Environment/Shell/Materials/Crew/M_UI_Crew_SceneBackground
graph {
 properties {pixelDepthOffsetMode=PDOM_Legacy, twoSided=true, usedShadingModels=Default Lit}
 nodes [
  NamedRerouteDeclaration(None=ScalarParameter() {defaultValue=1, parameterName=CrewLegacyUnsubscribed, sliderMax=1}) {name=Is Subscribed, nodeColor=<rgba 0 0 0 1>}
  n1 = NamedRerouteUsage()
  n2 = AppendVector(A=ScalarParameter() {parameterName=Radial Offset X}, B=ScalarParameter() {parameterName=Radial Offset Y})
  n3 = VectorParameter() {group=Background, parameterName=BGGradient, channelNames {a=Y, b=X, g=Density, r=Radius}, defaultValue=<rgba 0.8083 0.9559 0.5 0.5>}
  n4 = NamedRerouteDeclaration(None=MaterialFunctionCall(
    UVs=Add(A=@n1, B=@n2)
    Radius=@n3.Radius
    Density=@n3.Radius
    Invert Density=StaticBoolParameter() {defaultValue=true, group=Background, parameterName=InvertBG}
   ) {materialFunction=&RadialGradientExponential}.RadialGradientExponential) {name=Radial Mask, nodeColor=<rgba 0 0 0 1>}
  NamedRerouteDeclaration(None=OneMinus(None=@n4)) {name=Inverted Radial Mask, nodeColor=<rgba 0 0 0 1>}
  n5 = NamedRerouteUsage()
  n6 = MaterialFunctionCall(
   A=Constant3Vector() {constant=<rgba 0.2169 0.4572 0.7305 1>}
   B=Constant3Vector() {constant=<rgba 0.1014 0.2571 0.526 1>}
   C=Constant3Vector() {constant=<rgba 0.0009 0.0802 0.4564 1>}
   Alpha=@n5
  ) {materialFunction=&Lerp_3Color}
  n7 = MaterialFunctionCall(
   A=Constant3Vector() {constant=<rgba 0.9734 0.9216 0.3278 1>}
   B=Constant3Vector() {constant=<rgba 0.7305 0.5089 0.1248 1>}
   C=Constant3Vector() {constant=<rgba 0.227 0.0764 0.0003 1>}
   Alpha=@n5
  ) {materialFunction=&Lerp_3Color}
  n8 = NamedRerouteUsage()
  n9 = ScalarParameter() {defaultValue=1, parameterName=Outer Radial Brightness}
  n10 = Constant() {r=0.375}
  n11 = Add(A=@n10, B=Multiply(A=@n10, B=Constant() {r=0.6025}))
  n12 = VectorParameter() {group=Pattern, parameterName=PatternSpeed, defaultValue=<rgba 0.01 0 -0.05 0>}
  n13 = AppendVector(A=@n12.R, B=@n12.R)
  n14 = MaterialFunctionCall!(
   ScreenUV=NamedRerouteDeclaration(None=ScreenPosition().ViewportUV) {name=Base Screen UV, nodeColor=<rgba 0 0 0 1>}
   TextureSize
   ScalePivotPoint
  ) {materialFunction=/FortUILibrary/MaterialFunctions/Basics/MF_UI_ScreenSpaceUVScaling}
  n15 = ScalarParameter() {group=Pattern, parameterName=Rotation}
  n16 = VectorParameter() {group=Pattern, parameterName=TextureDimensions, defaultValue=<rgba 1 1 0 1>}
  n17 = Multiply(
   A=MaterialFunctionCall(
    UVs=MaterialFunctionCall(UVs=@n14.Scaled to Fit, "Rotation Angle (0-1)"=@n15) {materialFunction=&CustomRotator}.Rotated Values
    Texture Scale=AppendVector(A=@n16.R, B=@n16.R)
   ) {materialFunction=&ScaleUVsByCenter}.UVs
   B=ScalarParameter() {defaultValue=3, group=Pattern, parameterName=PatternRepeat}
  )
  n18 = TextureObject() {samplerType=SAMPLERTYPE_LinearColor, texture=&/T_UI_CrewPattern}
  n19 = AppendVector(A=@n12.B, B=@n12.B)
  n20 = Max(
   A=TextureSample(UVs=Panner(Coordinate=@n17, Speed=@n13), Tex=@n18) {samplerSource=SSM_Wrap_WorldGroupSettings}.R
   B=TextureSample(UVs=Panner(Coordinate=@n17, Speed=@n19), Tex=@n18) {samplerSource=SSM_Wrap_WorldGroupSettings}.G
  )
  n21 = Multiply(A=LinearInterpolate(Alpha=NamedRerouteUsage()) {constA=0.05, constB=0.4}, B=Saturate(None=SmoothStep(Min=@n10, Max=@n11, Value=@n20)))
  n22 = MaterialFunctionCall!(
   Base=Multiply(A=LinearInterpolate(A=@n6.Result, B=@n7.Result, Alpha=@n8), B=LinearInterpolate(A=@n9, Alpha=@n4))
   Add=LinearInterpolate(A=@n21, B=Multiply(A=@n21) {constB=-1}, Alpha=NamedRerouteUsage())
  ) {materialFunction=/Engine/Functions/Engine_MaterialFunctions02/Math/MultiplyAdd}
  NamedRerouteDeclaration(None=MaterialFunctionCall!(Color=Saturate(None=@n22.Result), Color Brightness, Floor Radius, Floor Density, Floor Metallic, Floor Specular, Floor Roughness) {materialFunction=/Game/UI/Foundation/Environment/Shell/Materials/MaterialFunctions/MF_UI_MPItemShop_Setup}.Floor Gradient Mask) {name=Floor Mask, nodeColor=<rgba 0 0 0 1>}
 ]
}
